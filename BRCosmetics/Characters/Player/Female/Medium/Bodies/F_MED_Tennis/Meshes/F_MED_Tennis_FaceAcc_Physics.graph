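# Physics asset for the face accessory mesh of the female medium Tennis cosmetic body, set up for hair and ponytail physics.
# Capsules on the upper spine, clavicles, neck and head form the body shapes, linked by free angular constraints.
# Two box-shaped hair bodies hang off the head with angular limits.
# A five-capsule ponytail chain hangs from a head-attached driver bone, with swing limits widening toward the tip.
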PhysicsAsset /BRCosmetics/Characters/Player/Female/Medium/Bodies/F_MED_Tennis/Meshes/F_MED_Tennis_FaceAcc_Physics
bodies {
 clavicle_l [{shapeType=Capsule, length=5, radius=7, center=<5.8861 -0.8118 -2.4179>, rotation=<rot 87.4608 0 -4.3713>}]
 clavicle_r [{shapeType=Capsule, length=6.5, radius=6.5, center=<-5.8116 1.3878 2.3031>, rotation=<rot 90 0 -9.0462>}]
 driver_dyn_ponytail_01 [{shapeType=Capsule, length=0.505, radius=0.505, rotation=<rot 0 -90 0>}]
 dyn_hair_01 [{shapeType=Box, extentX=3.51, extentY=1.01, extentZ=6.01, center=<0 0 -1.4772>, rotation=<rot 0 -90 0>}]
 dyn_hair_02 [{shapeType=Box, extentX=3.51, extentY=1.01, extentZ=6.01, center=<0 0 -1.4772>, rotation=<rot 0 -90 0>}]
 dyn_ponytail_01 [{shapeType=Capsule, length=1.755, radius=1.505, center=<0.8909 0 0>, rotation=<rot 90 0 0>}]
 dyn_ponytail_02 [{shapeType=Capsule, length=2.005, radius=1.255, center=<0.9116 0 0>, rotation=<rot 90 0 0>}]
 dyn_ponytail_03 [{shapeType=Capsule, length=2.005, radius=1.255, center=<0.8466 0 0>, rotation=<rot 90 0 0>}]
 dyn_ponytail_04 [{shapeType=Capsule, length=2.005, radius=1.255, center=<0.994 0 0>, rotation=<rot 90 0 0>}]
 dyn_ponytail_05 [{shapeType=Capsule, length=2.005, radius=1.255, center=<1.3247 0 0>, rotation=<rot 90 0 0>}]
 head [{shapeType=Capsule, length=3.1908, radius=5.602, rotation=<rot -90 0 30.003>}]
 neck_01 [{shapeType=Capsule, length=3.5, radius=5, center=<3.3298 0 0.0001>, rotation=<rot 90 0 0>}]
 spine_05 [{shapeType=Capsule, length=7, radius=14.755, center=<-0.0271 4.0261 0>, rotation=<rot 0 -90 0>}]
}
constraints [
 {
  bone1Name=head
  bone2Name=neck_01
  swing1LimitDegrees=45
  swing1Motion=Free
  swing2LimitDegrees=45
  swing2Motion=Free
  twistLimitDegrees=45
  twistMotion=Free
 }
 {bone1Name=dyn_hair_01, bone2Name=head, swing1LimitDegrees=25, swing2LimitDegrees=25, twistLimitDegrees=45}
 {bone1Name=dyn_ponytail_01, bone2Name=driver_dyn_ponytail_01, swing1LimitDegrees=45, swing2LimitDegrees=45, twistLimitDegrees=25}
 {bone1Name=dyn_ponytail_02, bone2Name=dyn_ponytail_01, swing1LimitDegrees=45, swing2LimitDegrees=45, twistLimitDegrees=20}
 {bone1Name=dyn_ponytail_03, bone2Name=dyn_ponytail_02, swing1LimitDegrees=45, swing2LimitDegrees=45, twistLimitDegrees=20}
 {bone1Name=dyn_ponytail_04, bone2Name=dyn_ponytail_03, swing1LimitDegrees=60, swing2LimitDegrees=60, twistLimitDegrees=20}
 {bone1Name=dyn_ponytail_05, bone2Name=dyn_ponytail_04, swing1LimitDegrees=60, swing2LimitDegrees=60, twistLimitDegrees=20}
 {bone1Name=dyn_hair_02, bone2Name=head, swing1LimitDegrees=25, swing2LimitDegrees=25, twistLimitDegrees=45}
 {
  bone1Name=neck_01
  bone2Name=spine_05
  swing1LimitDegrees=45
  swing1Motion=Free
  swing2LimitDegrees=45
  swing2Motion=Free
  twistLimitDegrees=45
  twistMotion=Free
 }
 {
  bone1Name=clavicle_l
  bone2Name=spine_05
  swing1LimitDegrees=45
  swing1Motion=Free
  swing2LimitDegrees=45
  swing2Motion=Free
  twistLimitDegrees=45
  twistMotion=Free
 }
 {
  bone1Name=clavicle_r
  bone2Name=spine_05
  swing1LimitDegrees=45
  swing1Motion=Free
  swing2LimitDegrees=45
  swing2Motion=Free
  twistLimitDegrees=45
  twistMotion=Free
 }
 {
  bone1Name=driver_dyn_ponytail_01
  bone2Name=head
  swing1LimitDegrees=45
  swing1Motion=Free
  swing2LimitDegrees=45
  swing2Motion=Free
  twistLimitDegrees=45
  twistMotion=Free
 }
]
