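# Post-process-style secondary motion AnimBP for the F_MED_Tennis face accessory skeleton, animating its ponytail on top of the pose and curves copied from the attached parent mesh.
# When bRigidBodyEnabled is true, AnimDynamics on the ponytail driver bone feeds a RigidBody simulation based at spine_05, both partially weighted.
# Otherwise, a Look At aims the ponytail bone at spine_05 as a non-simulated fallback.
# The event graph holds no logic.
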AnimBlueprint /BRCosmetics/Characters/Player/Female/Medium/Bodies/F_MED_Tennis/Meshes/F_MED_Tennis_FaceAcc_AnimBP
target_skeleton=&/F_MED_Tennis_FaceAcc_Skeleton
graphs [
 AnimGraph {nodes [
   n1 = Anim Dynamics - Bone: driver_dyn_ponytail_01(ComponentPose=Local To Component(LocalPose=Use cached pose 'AnimPose'()), Alpha=0.5, ExternalForce="0,0,0") {
    angularDampingOverride=0.995
    angularSpringConstant=250
    bAngularSpring=true
    bGravityOverrideInSimSpace=true
    bOverrideAngularDamping=true
    gravityScale=0.33
    simulationSpace=RootRelative
    alphaBoolBlend {blendOption=Linear}
    boundBone {boneName=driver_dyn_ponytail_01}
    componentAppliedLinearAccClamp=<100 100 100>
    componentLinearAccScale=<1 1 1>
    componentLinearVelScale=<1 1 1>
    physicsBodyDefinitions [{
      sphereCollisionRadius=0
      boundBone {boneName=driver_dyn_ponytail_01}
      boxExtents=<20 20 20>
      constraintSetup {twistAxis=AxisZ, angularLimitsMax=<25 25 10>, angularLimitsMin=<-15 -25 -10>, angularTarget=<0 0 0>}
      localJointOffset=<0 0 -10>
     }]
   }
   n2 = RigidBody(ComponentPose=@n1, Alpha=0.7) {
    bOverrideWorldGravity=true
    overridePhysicsAsset=&/F_MED_Tennis_FaceAcc_Physics
    simulationSpace=BaseBoneSpace
    alphaBoolBlend {blendOption=Linear}
    baseBoneRef {boneName=spine_05}
    componentAppliedLinearAccClamp=<10 10 10>
    componentLinearAccScale=<1 1 1>
    componentLinearVelScale=<1 1 1>
    overrideWorldGravity {z=-980}
   }
   n3 = Look At - Bone: dyn_ponytail_01(ComponentPose=Local To Component(LocalPose=Use cached pose 'AnimPose'()), Alpha=1) {
    alphaBoolBlend {blendOption=Linear}
    boneToModify {boneName=dyn_ponytail_01}
    lookAtLocation=<-15 -6.5>
    lookAtTarget {boneReference {boneName=spine_05}}
    lookAt_Axis {axis=<1 0>}
    lookUp_Axis {axis=<0 1>}
   }
   Output Pose(Result=Blend Poses by bool(BlendPose_0=Component To Local(ComponentPose=@n2), BlendPose_1=Component To Local(ComponentPose=@n3), bActiveValue=Get bRigidBodyEnabled()) {blendType=Linear}) {layerGroup=None}
   Save cached pose 'AnimPose'(Pose=Copy Pose From Mesh() {bCopyCurves=true, bUseAttachedParent=true}) {cacheName=AnimPose}
  ]}
 EventGraph {nodes [
   TryGetPawnOwner()
   Event BlueprintUpdateAnimation()
  ]}
]
variables []
